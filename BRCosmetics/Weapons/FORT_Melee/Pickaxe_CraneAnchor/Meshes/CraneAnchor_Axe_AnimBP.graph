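# Animation Blueprint for the CraneAnchor pickaxe that copies the mesh pose and continuously spins two wheel bones.
# Each update eases a spin speed at rates dependent on movement and melee state, accumulates an angle wrapped at 720, and applies it as negated additive roll.
# An Anim Dynamics or spring-driven suspension chain with paired Look At piston springs also exists, but does not feed the output pose.
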AnimBlueprint /BRCosmetics/Weapons/FORT_Melee/Pickaxe_CraneAnchor/Meshes/CraneAnchor_Axe_AnimBP
target_skeleton=&/CraneAnchor_Axe_Skeleton
graphs [
 AnimGraph {nodes [
   n1 = Anim Dynamics - Bone: dyn_carSuspension(Alpha=1, ExternalForce="0,0,0") {
    angularDampingOverride=0.99
    bOverrideAngularDamping=true
    bUseGravityOverride=true
    simulationSpace=BoneRelative
    alphaBoolBlend {blendOption=Linear}
    boundBone {boneName=dyn_carSuspension}
    componentLinearAccScale {y=0.5}
    componentLinearVelScale {y=0.25}
    gravityOverride {z=-980}
    physicsBodyDefinitions [{
      boundBone {boneName=dyn_carSuspension}
      boxExtents=<20 20 10>
      constraintSetup {angularLimitsMax=<0 15 0>, angularLimitsMin=<0 -15 0>, angularTarget=<0 0 0>}
      localJointOffset=<0 0 -7>
     }]
    relativeSpaceBone {boneName=b_AR_Root}
   }
   n2 = Component To Local(ComponentPose=@n1)
   n3 = Spring controller - Bone: dyn_carSuspension!(ComponentPose:Component Space Pose Link Structure=Local To Component(LocalPose=Copy Pose From Mesh()), Alpha:float=0.1) -> (Pose:Component Space Pose Link Structure) {
    is=title AnimGraphNode_SpringBone
    bLimitDisplacement=true
    bTranslateY=false
    bTranslateZ=false
    maxDisplacement=25
    alphaBoolBlend {blendOption=Linear}
    springBone {boneName=dyn_carSuspension}
   }
   n4 = Component To Local(ComponentPose=@n3)
   n5 = AND Boolean(A=Get bRigidBodyEnabled(), B=Get bAnimDynamicsEnabled() -> (bAnimDynamicsEnabled:bool))
   n6 = Look At - Bone: dyn_pistonSpring_lft_01(
    ComponentPose=Local To Component(LocalPose=Blend Poses by bool(BlendPose_0=@n2, BlendPose_1=@n4, bActiveValue=@n5) {blendType=Linear})
    Alpha=1
   ) {
    alphaBoolBlend {blendOption=Linear}
    boneToModify {boneName=dyn_pistonSpring_lft_01}
    lookAtLocation {x=0}
    lookAtTarget {boneReference {boneName=MSTR_pistonSpring_lft_00}}
    lookAt_Axis {axis=<-1 0>}
    lookUp_Axis {axis {x=0, z=1}}
   }
   n7 = Look At - Bone: dyn_pistonSpring_lft_00(ComponentPose=@n6, Alpha=1) {
    alphaBoolBlend {blendOption=Linear}
    boneToModify {boneName=dyn_pistonSpring_lft_00}
    lookAtLocation {x=0}
    lookAtTarget {boneReference {boneName=MSTR_pistonSpring_lft_01}}
    lookAt_Axis {axis=<1 0>}
    lookUp_Axis {axis {x=0, z=1}}
   }
   n8 = Look At - Bone: dyn_pistonSpring_rgt_01(ComponentPose=@n7, Alpha=1) {
    alphaBoolBlend {blendOption=Linear}
    boneToModify {boneName=dyn_pistonSpring_rgt_01}
    lookAtLocation {x=0}
    lookAtTarget {boneReference {boneName=MSTR_pistonSpring_rgt_00}}
    lookAt_Axis {axis=<1 0>}
    lookUp_Axis {axis {x=0, z=-1}}
   }
   Look At - Bone: dyn_pistonSpring_rgt_00(ComponentPose=@n8, Alpha=1) {
    alphaBoolBlend {blendOption=Linear}
    boneToModify {boneName=dyn_pistonSpring_rgt_00}
    lookAtLocation {x=0}
    lookAtTarget {boneReference {boneName=MSTR_pistonSpring_rgt_01}}
    lookAt_Axis {axis=<-1 0>}
    lookUp_Axis {axis {x=0, z=-1}}
   }
   n9 = Get CrankRotator() -> (CrankRotator:Rotator)
   n10 = "Transform (Modify) Bone - Bone: dyn_wheel_A"(ComponentPose=Local To Component(LocalPose=Copy Pose From Mesh()), Alpha=1, Translation="0,0,0", Rotation=@n9, Scale="1,1,1") {
    rotationMode=BMM_Additive
    rotationSpace=BCS_BoneSpace
    scaleMode=BMM_Additive
    scaleSpace=BCS_BoneSpace
    translationMode=BMM_Additive
    translationSpace=BCS_BoneSpace
    alphaBoolBlend {blendOption=Linear, blendOutTime=2}
    boneToModify {boneName=dyn_wheel_A}
   }
   n11 = "Transform (Modify) Bone - Bone: dyn_wheel_B"(ComponentPose=@n10, Alpha=1, Translation="0,0,0", Rotation=@n9, Scale="1,1,1") {
    rotationMode=BMM_Additive
    rotationSpace=BCS_BoneSpace
    scaleMode=BMM_Additive
    scaleSpace=BCS_BoneSpace
    translationMode=BMM_Additive
    translationSpace=BCS_BoneSpace
    alphaBoolBlend {blendOption=Linear, blendOutTime=2}
    boneToModify {boneName=dyn_wheel_B}
   }
   Output Pose(Result=Component To Local(ComponentPose=@n11)) {layerGroup=None}
   AnimGraphNode_SubInput_0 = Input Pose()
  ]}
 EventGraph {nodes [
   n1 = Event BlueprintUpdateAnimation()
   > n2 = Sequence!(execute) -> (then_0, then_1, then_2:Exec, then_3:Exec)
   > Set interpSpeed(interpSpeed:float=3) -> (Output_Get:float)
   n3 = Branch(
    execute=@n2.then_1
    Condition=OR Boolean!(A=Get bIsPlayingMeleeAnim(), B=Get bIsAccelerating2D(), C:bool=Get bIsJumping(), D:bool=Get bIsFalling()) -> (ReturnValue)
   )
   > Set interpSpeed(interpSpeed=11.2)
   Set interpSpeed(execute=@n3.else, interpSpeed=11.2)
   Set crackSpinSpeed(
    execute=@n2.then_2
    crackSpinSpeed:float=FInterpTo(Current=Get crackSpinSpeed() -> (crackSpinSpeed:float), Target=1, DeltaTime=@n1.DeltaTimeX, InterpSpeed=3)
   ) -> (Output_Get:float)
   > n4 = Sequence!(execute) -> (then_0, then_1)
   > n5 = Branch(Condition=OR Boolean!(A=Get bIsJumping(), B=Get bIsFalling(), C=Get bIsAccelerating2D()) -> (ReturnValue))
   > Set crackSpinSpeed(crackSpinSpeed=FInterpTo(Current=Get crackSpinSpeed(), Target=11.2, DeltaTime=@n1.DeltaTimeX, InterpSpeed=3))
   Set crackSpinSpeed(execute=[@n6.else, @n5.else], crackSpinSpeed=FInterpTo(Current=Get crackSpinSpeed(), Target=11.2, DeltaTime=@n1.DeltaTimeX, InterpSpeed=1))
   n6 = Branch(execute=@n4.then_1, Condition=Get bIsPlayingMeleeAnim())
   > Set crackSpinSpeed(crackSpinSpeed=FInterpTo(Current=Get crackSpinSpeed(), Target=11.2, DeltaTime=@n1.DeltaTimeX, InterpSpeed=3))
   n7 = SpinFunction!(execute=@n2.then_3, self, Spin:double=Get crankSpin() -> (crankSpin:double), Speed:double=Get crackSpinSpeed()) -> (then, Output:double)
   > n8 = Set crankSpin(crankSpin:double=@n7) -> (Output_Get:double)
   > Set CrankRotator(CrankRotator:Rotator=MakeRotator(Roll=float * float(A=@n8, B=-1) {is=title K2Node_PromotableOperator})) -> (Output_Get:Rotator)
   Get bIsPlayingMeleeAnim()
  ]}
 SpinFunction {nodes [
   n1 = SpinFunction!() -> (then, Spin:double=~, Speed:double=~) {is=title K2Node_FunctionEntry}
   > n2 = Set spinVar(spinVar:double=@n1) -> (Output_Get:double)
   > n3 = Branch(Condition="float >= float"(A=@n2, B=720) {is=title K2Node_CallFunction})
   > n4 = Set spinVar(spinVar=float - float(A=Get spinVar() -> (spinVar:double), B=720) {is=title K2Node_CallFunction})
   n5 = Get velocity() -> (velocity:double)
   n6 = Get velocityMult() -> (velocityMult:double)
   n7 = Get deltaTimeRemap() -> (deltaTimeRemap:double)
   n8 = Set spinVar(
    execute=@n3.else
    spinVar=float + float!(
     A=Get spinVar()
     B=float * float(A=@n1.Speed, B=@n7) {is=title K2Node_CommutativeAssociativeBinaryOperator}
     C:double=float * float(A=@n5, B=@n6)
    ) -> (ReturnValue) {is=title K2Node_CommutativeAssociativeBinaryOperator}
   )
   Return Node!(execute=[@n4.then, @n8.then], Output:double=@n8)
  ]}
]
variables [
 crackSpin {type=number}
 crackSpinSpeed {type=number}
 CrankRotator {type=Rotator}
 crankSpin {type=number, default=20.3125}
 deltaTimeRemap {type=number, default=7}
 interpSpeed {type=number}
 onCrackSpinSpeed {type=number}
 Spinning_Rotator_00 {type=Rotator}
 spinVar {type=number, default=20.3125}
 velocity {type=number}
 velocityMult {type=number}
]
